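# Sound cue for a pickaxe close-quarters warning from an enemy, judging by its name, built from melee impact sounds.
# A Random node picks among three equally weighted pickaxe CQC impact waves, then a Modulator varies the pitch slightly upward.
# It uses shared CQC impact attenuation and the melee weapon sound class.
# Overridden concurrency limits it to two voices per owner, stopping the oldest.
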SoundCue /Game/Athena/Sounds/Weapons/PickAxe_CQCWarning_Enemy_Cue
root=Modulator_1
nodes [
 {id=Modulator_1, children [Random_0], properties {pitchMax=1.2, pitchMin=1}}
 {id=Random_0, children [WavePlayer_0,WavePlayer_1,WavePlayer_2], properties {weights [1,1,1]}}
 {id=WavePlayer_0, properties {soundWaveAssetPtr=&/PickAxe_CQC_Impact_01}}
 {id=WavePlayer_1, properties {soundWaveAssetPtr=&/PickAxe_CQC_Impact_02}}
 {id=WavePlayer_2, properties {soundWaveAssetPtr=&/PickAxe_CQC_Impact_03}}
]
properties {
 attenuationSettings=&/ATT_CQC_Impacts
 bOverrideConcurrency=true
 maxDistance=1024
 soundClassObject=&Fort_WeaponMelee_PSM
 volumeMultiplier=1
 attenuationOverrides {
  bEnableReverbSend=false
  falloffDistance=512
  spatializationAlgorithm=SPATIALIZATION_HRTF
  stereoSpread=0
  attenuationShapeExtents=<512 512 512>
 }
 concurrencyOverrides {bLimitToOwner=true, maxCount=2, resolutionRule=StopOldest}
 concurrencySet [&Player_Melee_Weapon_Conc]
}
