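# Sound cue for enemy close-quarters melee impacts, apparently a warning sound going by its name, built from axe world-impact recordings.
# A Random node picks one of five equally weighted waves, and a Modulator varies pitch upward before output.
# It uses the CQC impacts attenuation asset and the impacts sound class.
# Overridden concurrency caps simultaneous instances per owner and stops the oldest.
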SoundCue /Game/Athena/Sounds/Weapons/Weapon_Impact_CQCWarning_Enemy_Cue
root=Modulator_1
nodes [
 {id=Modulator_1, children [Random_5], properties {pitchMax=1.2, pitchMin=1}}
 {id=Random_5, children [WavePlayer_5, WavePlayer_21, WavePlayer_22, WavePlayer_23, WavePlayer_24], properties {weights [1,1,1,1,1]}}
 {id=WavePlayer_21, properties {soundWaveAssetPtr=/Game/Sounds/Fort_Weapons/Axes/Claxe_WorldImpact_004}}
 {id=WavePlayer_22, properties {soundWaveAssetPtr=/Game/Sounds/Fort_Weapons/Axes/Claxe_WorldImpact_005}}
 {id=WavePlayer_23, properties {soundWaveAssetPtr=/Game/Sounds/Fort_Weapons/Axes/Claxe_WorldImpact_001}}
 {id=WavePlayer_24, properties {soundWaveAssetPtr=/Game/Sounds/Fort_Weapons/Axes/Claxe_WorldImpact_003}}
 {id=WavePlayer_5, properties {soundWaveAssetPtr=/Game/Sounds/Fort_Weapons/Axes/Claxe_WorldImpact_002}}
]
properties {
 attenuationSettings=&/ATT_CQC_Impacts
 bOverrideConcurrency=true
 maxDistance=1024
 soundClassObject=&Fort_Impacts
 volumeMultiplier=1.25
 attenuationOverrides {
  attenuationShape=Cone
  bEnableReverbSend=false
  coneOffset=50
  falloffDistance=175
  spatializationAlgorithm=SPATIALIZATION_HRTF
  stereoSpread=0
  attenuationShapeExtents=<425.9987 73.2207 -20.4545>
 }
 concurrencyOverrides {bLimitToOwner=true, maxCount=3, resolutionRule=StopOldest}
 concurrencySet [&Player_Melee_Weapon_Conc]
}
